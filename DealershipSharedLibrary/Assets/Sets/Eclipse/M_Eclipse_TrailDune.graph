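# Translucent, unlit, two-sided trail material for Niagara ribbons, named for the Eclipse set's TrailDune effect.
# Two panning fire-noise layers, masked along length and width, distort the UVs of painted and diffuse trail textures.
# Dynamic parameters blend channels and add edge glow, while starburst textures add length-masked sparkles.
# Particle color tints the sum through MF_FX_EmissiveAdjustments; particle alpha scales opacity from the trail masks.
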Material /DealershipSharedLibrary/Assets/Sets/Eclipse/M_Eclipse_TrailDune
graph {
 properties {
  bAutomaticallySetUsageInEditor=false
  bUsedWithNiagaraRibbons=true
  blendMode=BLEND_Translucent
  shadingModel=MSM_Unlit
  twoSided=true
  usedShadingModels=Unlit
 }
 nodes [
  n1 = ParticleColor()
  n2 = TextureCoordinate()
  n3 = Multiply(A=Constant() {r=2.5}, B=OneMinus(None=ComponentMask(None=@n2) {r=true}))
  n4 = Power(
   Base=TextureSample(UVs=TextureCoordinate() {uTiling=1.5, vTiling=1.5}) {texture=/DealershipSharedLibrary/FX/Textures/TexturePackage/T_VCVFX_Starbursts}.B
   Exp=Constant() {r=2.7}
  )
  n5 = TextureSample(UVs=Panner(Coordinate=TextureCoordinate() {uTiling=2}) {speedX=-1.2}) {texture=&T_VCVFX_Starbursts}
  n6 = Multiply(A=@n5.G, B=Constant())
  n7 = Multiply(A=Constant() {r=6}, B=ComponentMask(None=TextureCoordinate()) {r=true})
  n8 = Constant() {r=2}
  n9 = Multiply(A=Saturate(None=Power(Base=@n7, Exp=@n8)), B=@n5.G)
  n10 = DynamicParameter() {paramNames [Param1,Param2,Invert,Param4]}
  n11 = LinearInterpolate(A=Multiply(A=@n9, B=Constant() {r=10}), B=Multiply(A=@n9, B=Constant() {r=0.75}), Alpha=@n10.Param2)
  n12 = OneMinus(None=ComponentMask(None=TextureCoordinate()) {r=true})
  n13 = Multiply(
   A=Multiply(A=Multiply(A=@n3, B=@n4), B=LinearInterpolate(A=@n6, B=@n11, Alpha=@n10.Invert))
   B=Round(None=Power(Base=@n12) {constExponent=0.45})
  )
  n14 = Time()
  n15 = Constant() {r=1}
  n16 = TextureCoordinate() {uTiling=1.5, vTiling=0.25}
  n17 = Add(
   A=Panner(Coordinate=TextureCoordinate() {uTiling=0.6}, Time=Multiply(A=@n14, B=@n15)) {speedY=0.5}
   B=TextureSample(UVs=Panner(Coordinate=@n16) {speedX=0.15}) {texture=/DealershipSharedLibrary/FX/Textures/TexturePackage/T_VCVFX_FireNoise}.G
  )
  n18 = Time()
  n19 = Constant() {r=1}
  n20 = TextureCoordinate() {vTiling=0.1}
  n21 = Add(
   A=Panner(Coordinate=TextureCoordinate() {uTiling=0.6}, Time=Multiply(A=@n18, B=@n19)) {speedY=-0.26}
   B=TextureSample(UVs=Panner(Coordinate=@n20) {speedX=0.2}) {texture=&T_VCVFX_FireNoise}.G
  )
  n22 = Multiply(
   A=Add(
    A=TextureSample(UVs=@n17) {texture=&T_VCVFX_FireNoise}.R
    B=TextureSample(UVs=@n21) {texture=&T_VCVFX_FireNoise}.R
   )
   B=Constant() {r=0.25}
  )
  n23 = TextureCoordinate()
  n24 = ComponentMask(None=@n23) {r=true}
  n25 = Constant() {r=2}
  n26 = ComponentMask(None=@n23) {g=true}
  n27 = Constant() {r=0.8}
  n28 = Multiply(A=Multiply(A=Power(Base=@n24, Exp=@n25), B=Multiply(A=@n26, B=@n27)), B=Multiply(A=OneMinus(None=@n26), B=@n27))
  n29 = Saturate(None=Add(A=TextureCoordinate(), B=Multiply(A=@n22, B=@n28)))
  n30 = TextureSample(UVs=@n29) {texture=/DealershipSharedLibrary/FX/Textures/TexturePackage/T_VCVFX_TrailPainted}
  n31 = TextureSample(UVs=@n29) {texture=/DealershipSharedLibrary/FX/Textures/Diffuse/T_VCVFX_Trail}
  n32 = LinearInterpolate(A=@n30.R, B=@n31.RGB, Alpha=@n10.Param1)
  n33 = Multiply(A=@n30.R, B=DynamicParameter() {parameterIndex=1, paramNames [EDGE EMISSIVENESS, Param2, Invert, Param4]}.EDGE EMISSIVENESS)
  n34 = Add(
   A=LinearInterpolate(A=@n30.B, B=LinearInterpolate(A=@n30.G, B=@n32, Alpha=@n10.Invert), Alpha=@n10.Param4)
   B=Saturate(None=SmoothStep(Value=@n33) {constMin=0.5})
  )
  MaterialFunctionCall(Emissive=Multiply(A=@n1.RGB, B=Add(A=@n13, B=@n34))) {materialFunction=&MF_FX_EmissiveAdjustments}
  Multiply(A=@n1.A, B=Add(A=@n31.A, B=@n30.R))
 ]
}
